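; Sandbag prop Blueprint, apparently a child of a building actor class given its bIsInvulnerable variable.
; On BeginPlay it runs the parent logic, then sets bIsInvulnerable to true so the sandbags are flagged invulnerable.
; The construction script is empty.
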(event EventBeginPlay
  (|CallParentFunction)
  (Variables|BuildingActor|SetbIsInvulnerable true))

(fn ConstructionScript)
